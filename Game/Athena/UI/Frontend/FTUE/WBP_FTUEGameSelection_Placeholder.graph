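# Placeholder first-time user experience mode selection screen, derived from FortFirstTimeModeSelectionScreen, with two Play buttons and a Close button over a static background image.
# On Construct, the async GetFTUEPlaylists action fills GameModes, writing the result on both completion and failure.
# Each Play button checks the array length, then passes the first or second mode and the owning local player to OnPlayButtonPressed.
# Close deactivates the widget.
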WidgetBlueprint /Game/Athena/UI/Frontend/FTUE/WBP_FTUEGameSelection_Placeholder
graphs [EventGraph {nodes [
   "On Pressed (Button_Close)"!() -> (OutputDelegate, then, Button:Common Button Base Object Reference) {is=title K2Node_ComponentBoundEvent}
   > DeactivateWidget()
   "On Pressed (ButtonPlayMode1)"() {is=title K2Node_ComponentBoundEvent}
   n1 = Get GameModes() -> (GameModes:Array of FTUEGame Mode Structures)
   > Branch(Condition="integer >= integer"(A=Length(TargetArray:Array of FTUEGame Mode Structures=@n1), B=1))
   > OnPlayButtonPressed!(
    execute
    self:Fort First Time Mode Selection Screen Object Reference
    LocalPlayer:Local Player Object Reference=GetOwningLocalPlayer!(self:Widget Object Reference) -> (ReturnValue:Local Player Object Reference)
    GameMode:"FTUEGame Mode Structure (by ref)"="Get (a copy)"(Array:Array of FTUEGame Mode Structures=@n1) -> (Output:FTUEGame Mode Structure)
   ) -> (then)
   "On Pressed (ButtonPlayMode2)"() {is=title K2Node_ComponentBoundEvent}
   n2 = Get GameModes()
   > Branch(Condition="integer >= integer"(A=Length(TargetArray=@n2), B=2))
   > OnPlayButtonPressed(LocalPlayer=GetOwningLocalPlayer(), GameMode="Get (a copy)"(Array=@n2, Dimension 1=1))
   Event Construct()
   > n3 = GetFTUEPlaylists!(execute:Exec, LocalPlayer:Local Player Object Reference=GetOwningLocalPlayer()) -> (then:Exec, OnCompleted:Exec, OnFailed:Exec, GameModes:Array of FTUEGame Mode Structures) {is=title K2Node_AsyncAction}
   Set GameModes(execute=[@n3.OnCompleted, @n3.OnFailed], GameModes:Array of FTUEGame Mode Structures=@n3) -> (Output_Get:Array of FTUEGame Mode Structures)
  ]}]
variables [GameModes {type="array<FTUEGameMode>", default []}]
widgets {
 info {inheritedWidgetCount=0, parentClass=/Script/FortniteUI.FortFirstTimeModeSelectionScreen, rootWidgetClass=&Overlay, widgetCount=7}
 widgets [
  Overlay_0 {widgetClassPath=&Overlay, widget=Overlay_0}
  Background {
   widgetClassPath=&Image
   bIsVariable=true
   parent=Overlay_0
   widget=Background
   properties {brush {
     bIsDynamicallyLoaded=false
     drawAs=Image
     imageType=FullColor
     mirroring=NoMirror
     resourceName=None
     resourceObject=&/placeholderbg1
     tiling=NoTile
     imageSize=<1945 1095>
     margin=<ltrb 0 0 0 0>
     outlineSettings {
      bUseBrushTransparency=false
      roundingType=HalfHeightRadius
      width=0
      color {colorUseRule=UseColor_Specified, specifiedColor=<rgba 0 0 0 0>}
      cornerRadii=<0 0 0 0>
     }
     tintColor {colorUseRule=UseColor_Specified, specifiedColor=<rgba 1 1 1 1>}
     uVRegion {bIsValid=false, max=<0 0>, min=<0 0>}
    }}
   slot {class=OverlaySlot, properties {horizontalAlignment=HAlign_Fill, verticalAlignment=VAlign_Fill}}
  }
  HorizontalBox_1 {
   widgetClassPath=&HorizontalBox
   parent=Overlay_0
   widget=HorizontalBox_1
   slot {class=OverlaySlot, properties {horizontalAlignment=HAlign_Fill, verticalAlignment=VAlign_Bottom}}
  }
  Button_Close {
   widgetClassPath=&WBP_UIKit_Button_Regular
   bIsVariable=true
   parent=HorizontalBox_1
   widget=Button_Close
   properties {
    buttonSize=Small
    buttonType=Quiet
    minWidth=270
    navigation=&self:WidgetTree.Button_Close.WidgetNavigation_0
    text=Close
   }
   slot {class=HorizontalBoxSlot, properties {padding=<ltrb 800 0 0 0>, size {sizeRule=Automatic, value=0.33}}}
  }
  HorizontalBox {
   widgetClassPath=&HorizontalBox
   parent=Overlay_0
   widget=HorizontalBox
   slot {class=OverlaySlot, properties {horizontalAlignment=HAlign_Fill, verticalAlignment=VAlign_Center, padding=<ltrb 0 185 0 0>}}
  }
  ButtonPlayMode1 {
   widgetClassPath=&WBP_UIKit_Button_Generic
   bIsVariable=true
   parent=HorizontalBox
   widget=ButtonPlayMode1
   properties {
    backgroundMaterial=/FortUILibrary/UIKit/Buttons/MaterialInstances/CTAButtons/MI_UIKit_BgLoud
    outlineMaterial=/FortUILibrary/UIKit/Buttons/MaterialInstances/TileOutline/MI_UIKit_Outline
    text=Play
   }
   slot {class=HorizontalBoxSlot, properties {horizontalAlignment=HAlign_Center, verticalAlignment=VAlign_Center, size {sizeRule=Fill, value=1}}}
  }
  ButtonPlayMode2 {
   widgetClassPath=&WBP_UIKit_Button_Generic
   bIsVariable=true
   parent=HorizontalBox
   widget=ButtonPlayMode2
   properties {
    backgroundMaterial=&MI_UIKit_BgLoud
    outlineMaterial=&MI_UIKit_Outline
    text=Play
   }
   slot {class=HorizontalBoxSlot, properties {horizontalAlignment=HAlign_Center, verticalAlignment=VAlign_Center, size {sizeRule=Fill, value=1}}}
  }
 ]
}
